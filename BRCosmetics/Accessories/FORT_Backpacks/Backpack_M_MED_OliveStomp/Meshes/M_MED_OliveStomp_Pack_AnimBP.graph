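# Animation Blueprint for the OliveStomp backpack cosmetic, animating its pack, paper and corner bones.
# With anim dynamics enabled, chained AnimDynamics nodes with gravity override swing the paper; otherwise a trail controller does, both layered with additive bone offsets.
# The event graph derives pack sway and melee or aiming alphas from owner spine and arm deltas, and feeds WPO material functions.
# Debug prints remain, and several computed variables go unused.
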AnimBlueprint /BRCosmetics/Accessories/FORT_Backpacks/Backpack_M_MED_OliveStomp/Meshes/M_MED_OliveStomp_Pack_AnimBP
target_skeleton=&/M_MED_OliveStomp_Pack_Skeleton
graphs [
 AnimGraph {nodes [
   n1 = "Transform (Modify) Bone - Bone: dyn_L_corner"(Alpha=1, Translation="0,0,0", Rotation="0,0,-45", Scale="1,1,1") {
    rotationMode=BMM_Additive
    rotationSpace=BCS_BoneSpace
    translationMode=BMM_Additive
    translationSpace=BCS_BoneSpace
    alphaBoolBlend {blendInTime=0.3, blendOption=Linear, blendOutTime=0.3}
    alphaScaleBiasClamp {bInterpResult=true, interpSpeedDecreasing=5, interpSpeedIncreasing=5}
    boneToModify {boneName=dyn_L_corner}
    rotation {roll=-45}
   }
   "Transform (Modify) Bone - Bone: dyn_R_corner"(ComponentPose=@n1, Alpha=1, Translation="0,0,0", Rotation="0,0,45", Scale="1,1,1") {
    rotationMode=BMM_Additive
    rotationSpace=BCS_BoneSpace
    scaleMode=BMM_Additive
    scaleSpace=BCS_BoneSpace
    translationMode=BMM_Additive
    translationSpace=BCS_BoneSpace
    alphaBoolBlend {blendOption=Linear}
    alphaScaleBiasClamp {bInterpResult=true, interpSpeedDecreasing=5, interpSpeedIncreasing=5}
    boneToModify {boneName=dyn_R_corner}
    rotation {roll=45}
   }
   n2 = "Transform (Modify) Bone - Bone: pack"(ComponentPose=Local To Component(LocalPose=Use cached pose 'Base0'()), Translation="1,-1,0.5", Rotation="0,0,0") {
    translationMode=BMM_Additive
    translationSpace=BCS_BoneSpace
    alphaBoolBlend {blendOption=Linear}
    boneToModify {boneName=pack}
    translation=<1 -1 0.5>
   }
   n3 = Trail controller - Bone: dyn_paper_03(ComponentPose=@n2, Alpha=1) {
    bLimitRotation=true
    bShowTrailLocation=true
    bUsePlanarLimit=true
    baseJoint {boneName=pack}
    planarLimits [{drivingBone {boneName=dyn_paper_base}, planeTransform {location=<0 2 0>, rotation=<rot 0 0 -85>, scale=<1 1 1>}}]
    relaxationSpeedScaleInputProcessor {bInterpResult=true, interpSpeedDecreasing=5, interpSpeedIncreasing=5}
    rotationLimits [{limitMax=<0 0 -5>, limitMin=<0 0 -20>}, {limitMax=<0 0 -5>, limitMin=<0 0 -20>}]
    rotationOffsets [<0 0 0>, <0 0 0>]
    trailBone {boneName=dyn_paper_03}
    trailRelaxationSpeed {editorCurveData {keys [<at 0.0137 6.1005 RCIM_Linear>,<at 0.6642 8.5621 RCIM_Linear>,<at 1 5 RCIM_Linear>]}}
   }
   n4 = "Transform (Modify) Bone - Bone: dyn_paper_02"!(ComponentPose=@n3, bAlphaBoolEnabled:bool=True, Translation="0,0,0", Rotation="0,-20,0", Scale="1,1,1") -> (Pose) {
    alphaInputType=Bool
    rotationMode=BMM_Additive
    rotationSpace=BCS_BoneSpace
    translationMode=BMM_Additive
    translationSpace=BCS_BoneSpace
    alphaBoolBlend {
     blendInTime=3
     blendOption=Custom
     blendOutTime=3
     customCurve=&/M_MED_OliveStomp_Curve
    }
    boneToModify {boneName=dyn_paper_02}
    rotation {yaw=-20}
   }
   n5 = "Transform (Modify) Bone - Bone: pack"!(ComponentPose=@n4, bAlphaBoolEnabled:bool, Translation="0,-1.2,0", Rotation="0,0,0", Scale) -> (Pose) {
    alphaInputType=Bool
    translationMode=BMM_Additive
    translationSpace=BCS_BoneSpace
    alphaBoolBlend {blendOption=Linear}
    boneToModify {boneName=pack}
    translation {y=-1.2}
   }
   n6 = "Transform (Modify) Bone - Bone: dyn_L_corner"!(ComponentPose=@n5, bAlphaBoolEnabled:bool=True, Translation="0,0,0", Rotation="0,0,-15", Scale="1,1,1") -> (Pose) {
    alphaInputType=Bool
    rotationMode=BMM_Additive
    rotationSpace=BCS_BoneSpace
    translationMode=BMM_Additive
    translationSpace=BCS_BoneSpace
    alphaBoolBlend {blendInTime=0.3, blendOption=Sinusoidal, blendOutTime=0.3}
    boneToModify {boneName=dyn_L_corner}
    rotation {roll=-15}
   }
   n7 = "Transform (Modify) Bone - Bone: dyn_paper_03"!(ComponentPose=@n6, bAlphaBoolEnabled:bool=True, Translation="0,0,0", Rotation="0,-50,0", Scale="1,1,1") -> (Pose) {
    alphaInputType=Bool
    rotationMode=BMM_Additive
    rotationSpace=BCS_BoneSpace
    translationMode=BMM_Additive
    translationSpace=BCS_BoneSpace
    alphaBoolBlend {blendInTime=0.3, blendOption=Sinusoidal, blendOutTime=0.3}
    boneToModify {boneName=dyn_paper_03}
    rotation {yaw=-50}
   }
   n8 = "Transform (Modify) Bone - Bone: dyn_paper_base"(ComponentPose=@n7, Alpha=Get On Melee Right Extra() -> (On Melee Right Extra:double), Translation="0,0,0", Rotation="10,0,0", Scale="1,1,1") {
    rotationMode=BMM_Additive
    rotationSpace=BCS_BoneSpace
    scaleMode=BMM_Additive
    scaleSpace=BCS_BoneSpace
    translationMode=BMM_Additive
    translationSpace=BCS_BoneSpace
    alphaBoolBlend {blendOption=Linear}
    alphaScaleBiasClamp {bInterpResult=true, interpSpeedDecreasing=5, interpSpeedIncreasing=5}
    boneToModify {boneName=dyn_paper_base}
    rotation {pitch=10}
   }
   n9 = "Transform (Modify) Bone - Bone: dyn_L_corner"!(ComponentPose=@n8, Alpha:float=1, Translation="0,0,0", Rotation="0,0,-25", Scale="1,1,1") -> (Pose) {
    rotationMode=BMM_Additive
    rotationSpace=BCS_BoneSpace
    translationMode=BMM_Additive
    translationSpace=BCS_BoneSpace
    alphaBoolBlend {blendInTime=0.3, blendOption=Linear, blendOutTime=0.3}
    alphaScaleBiasClamp {bInterpResult=true, interpSpeedDecreasing=5, interpSpeedIncreasing=5}
    boneToModify {boneName=dyn_L_corner}
    rotation {roll=-25}
   }
   n10 = "Transform (Modify) Bone - Bone: dyn_R_corner"(ComponentPose=@n9, Alpha=1, Translation="0,0,0", Rotation="0,0,45", Scale="1,1,1") {
    rotationMode=BMM_Additive
    rotationSpace=BCS_BoneSpace
    scaleMode=BMM_Additive
    scaleSpace=BCS_BoneSpace
    translationMode=BMM_Additive
    translationSpace=BCS_BoneSpace
    alphaBoolBlend {blendOption=Linear}
    alphaScaleBiasClamp {bInterpResult=true, interpSpeedDecreasing=5, interpSpeedIncreasing=5}
    boneToModify {boneName=dyn_R_corner}
    rotation {roll=45}
   }
   n11 = "Transform (Modify) Bone - Bone: dyn_paper_02"!(ComponentPose=@n10, Alpha:float=1, Translation="0,0,0", Rotation="0,-25,0", Scale="1,1,1") -> (Pose) {
    rotationMode=BMM_Additive
    rotationSpace=BCS_BoneSpace
    scaleMode=BMM_Additive
    scaleSpace=BCS_BoneSpace
    translationMode=BMM_Additive
    translationSpace=BCS_BoneSpace
    alphaBoolBlend {blendOption=Linear}
    alphaScaleBiasClamp {interpSpeedDecreasing=5, interpSpeedIncreasing=5}
    boneToModify {boneName=dyn_paper_02}
    rotation {yaw=-25}
   }
   n12 = "Transform (Modify) Bone - Bone: dyn_paper_01"(ComponentPose=@n11, Alpha=1, Translation="0,0,0", Rotation="0,-20,0", Scale="1,1,1") {
    rotationMode=BMM_Additive
    rotationSpace=BCS_BoneSpace
    alphaBoolBlend {blendOption=Linear}
    alphaScaleBiasClamp {bInterpResult=true, interpSpeedDecreasing=5, interpSpeedIncreasing=5}
    boneToModify {boneName=dyn_paper_01}
    rotation {yaw=-20}
   }
   n13 = "Transform (Modify) Bone - Bone: dyn_paper_01"(ComponentPose=@n12, Alpha=1, Translation="0,0,0", Rotation="0,-15,0", Scale="1,1,1") {
    rotationMode=BMM_Additive
    rotationSpace=BCS_BoneSpace
    alphaBoolBlend {blendOption=Linear}
    alphaScaleBiasClamp {bInterpResult=true, interpSpeedDecreasing=5, interpSpeedIncreasing=5}
    boneToModify {boneName=dyn_paper_01}
    rotation {yaw=-15}
   }
   n14 = "Transform (Modify) Bone - Bone: dyn_paper_01"(ComponentPose=@n13, Alpha=Get Targeting_lookDown() -> (Targeting_lookDown:double), Translation="0,0,0", Rotation="0,-10,0", Scale="1,1,1") {
    rotationMode=BMM_Additive
    rotationSpace=BCS_BoneSpace
    alphaBoolBlend {blendOption=Linear}
    alphaScaleBiasClamp {bInterpResult=true, interpSpeedDecreasing=5, interpSpeedIncreasing=5}
    boneToModify {boneName=dyn_paper_01}
    rotation {yaw=-10}
   }
   n15 = "Transform (Modify) Bone - Bone: dyn_paper_01"(ComponentPose=@n14, Alpha=Get Targeting_lookUp() -> (Targeting_lookUp:double), Translation="0,0,0", Rotation="0,-10,0", Scale="1,1,1") {
    rotationMode=BMM_Additive
    rotationSpace=BCS_BoneSpace
    alphaBoolBlend {blendOption=Linear}
    alphaScaleBiasClamp {bInterpResult=true, interpSpeedDecreasing=5, interpSpeedIncreasing=5}
    boneToModify {boneName=dyn_paper_01}
    rotation {yaw=-10}
   }
   n16 = "Transform (Modify) Bone - Bone: pack"(ComponentPose=@n15, Translation="0,0.5,0", Rotation="0,0,-5") {
    alphaInputType=Bool
    rotationMode=BMM_Additive
    rotationSpace=BCS_BoneSpace
    translationMode=BMM_Additive
    translationSpace=BCS_BoneSpace
    alphaBoolBlend {blendOption=Linear}
    boneToModify {boneName=pack}
    rotation {roll=-5}
    translation {y=0.5}
   }
   Output Pose(Result=Blend Poses by bool(
     BlendPose_0=Use cached pose 'BaseMovement'()
     BlendPose_1=Component To Local(ComponentPose=@n16)
     bActiveValue=Get bAnimDynamicsEnabled() -> (bAnimDynamicsEnabled:bool)
    ) {blendType=Linear}) {layerGroup=None}
   n17 = "Transform (Modify) Bone - Bone: pack"!(ComponentPose=Local To Component(LocalPose=Use cached pose 'Base0'()), Alpha:float, Translation="1,0,0", Rotation="0,0,0", Scale) -> (Pose) {
    translationMode=BMM_Additive
    translationSpace=BCS_BoneSpace
    alphaBoolBlend {blendOption=Linear}
    boneToModify {boneName=pack}
    translation {x=1}
   }
   n18 = "Transform (Modify) Bone - Bone: pack"(ComponentPose=@n17, Translation=Get PelvisLocation() -> (PelvisLocation:Vector), Rotation=Get Backpack_Rotate() -> (Backpack_Rotate:Rotator)) {
    rotationMode=BMM_Additive
    rotationSpace=BCS_BoneSpace
    translationMode=BMM_Additive
    translationSpace=BCS_BoneSpace
    alphaBoolBlend {blendOption=QuadraticInOut}
    boneToModify {boneName=pack}
   }
   n19 = Get Gravity Override()
   n20 = Anim Dynamics - Bone: dyn_paper_01!(ComponentPose=@n18, Alpha=0.9, GravityOverride:Vector=@n19, ExternalForce="0,0,0") -> (Pose) {
    angularDampingOverride=0.999
    angularSpringConstant=50
    bAngularSpring=true
    bOverrideAngularDamping=true
    bUseGravityOverride=true
    bUseSphericalLimits=true
    simulationSpace=RootRelative
    alphaBoolBlend {blendOption=Linear}
    boundBone {boneName=dyn_paper_01}
    componentAppliedLinearAccClamp=<2000 2000 2000>
    componentLinearAccScale=<0.4 0.4 0.4>
    componentLinearVelScale=<0.4 0.4 0.4>
    physicsBodyDefinitions [{
      boundBone {boneName=dyn_paper_01}
      boxExtents=<10 10 10>
      constraintSetup {angularLimitsMax=<0 0 -2>, angularLimitsMin=<0 0 -30>, angularTarget=<0 0 0>}
      localJointOffset=<1 0 0>
     }]
    planarLimits [{drivingBone {boneName=pack}, planeTransform {location=<0 1.1436 -0.0707>, rotation=<rot 0 0 -86.4629>}}]
    sphericalLimits [{limitRadius=21.5772, drivingBone {boneName=pack}, sphereLocalOffset=<0 22.7785 -1.7936>}]
   }
   n21 = Anim Dynamics - Bone: dyn_paper_02!(ComponentPose=@n20, Alpha=0.9, GravityOverride:Vector=@n19, ExternalForce="0,0,0") -> (Pose) {
    angularDampingOverride=0.999
    angularSpringConstant=50
    bAngularSpring=true
    bOverrideAngularDamping=true
    bUseGravityOverride=true
    bUseSphericalLimits=true
    simulationSpace=RootRelative
    alphaBoolBlend {blendOption=Linear}
    boundBone {boneName=dyn_paper_02}
    componentAppliedLinearAccClamp=<2000 2000 2000>
    componentLinearAccScale=<0.4 0.4 0.4>
    componentLinearVelScale=<0.4 0.4 0.4>
    physicsBodyDefinitions [{
      boundBone {boneName=dyn_paper_02}
      boxExtents=<10 10 10>
      constraintSetup {angularLimitsMax=<0 0 -2>, angularLimitsMin=<0 0 -30>, angularTarget=<0 0 0>}
      localJointOffset=<1 0 0>
     }]
    planarLimits [{drivingBone {boneName=pack}, planeTransform {location=<-0.1377 0.6638 -0.0582>, rotation=<rot 0 0 -88.6742>}}]
    sphericalLimits [{limitRadius=23.3575, drivingBone {boneName=pack}, sphereLocalOffset=<0 23.2034 -5.0064>}]
   }
   n22 = Anim Dynamics - Bone: dyn_paper_03!(ComponentPose=@n21, Alpha=0.9, GravityOverride:Vector=@n19, ExternalForce="0,0,0") -> (Pose) {
    angularDampingOverride=0.999
    angularSpringConstant=50
    bAngularSpring=true
    bOverrideAngularDamping=true
    bUseGravityOverride=true
    bUseSphericalLimits=true
    simulationSpace=RootRelative
    alphaBoolBlend {blendOption=Linear}
    boundBone {boneName=dyn_paper_03}
    componentAppliedLinearAccClamp=<2000 2000 2000>
    componentLinearAccScale=<0.4 0.4 0.4>
    componentLinearVelScale=<0.4 0.4 0.4>
    physicsBodyDefinitions [{
      boundBone {boneName=dyn_paper_03}
      boxExtents=<10 10 10>
      constraintSetup {angularLimitsMax=<0 0 -2>, angularLimitsMin=<0 0 -30>, angularTarget=<0 0 0>}
      localJointOffset=<1 0 0>
     }]
    planarLimits [{drivingBone {boneName=pack}, planeTransform {location=<0 1.2638 -0.0781>, rotation=<rot 0 0 -86.4629>}}]
    sphericalLimits [{limitRadius=22.6245, drivingBone {boneName=pack}, sphereLocalOffset=<0 21.3713 -18.9857>}]
   }
   n23 = "Transform (Modify) Bone - Bone: pack"!(ComponentPose=@n22, bAlphaBoolEnabled:bool, Translation="0,1.2,0", Rotation="0,0,0", Scale) -> (Pose) {
    alphaInputType=Bool
    translationMode=BMM_Additive
    translationSpace=BCS_BoneSpace
    alphaBoolBlend {blendInTime=0.1, blendOption=Linear, blendOutTime=0.1}
    boneToModify {boneName=pack}
    translation {y=1.2}
   }
   n24 = "Transform (Modify) Bone - Bone: dyn_paper_02"!(ComponentPose=@n23, bAlphaBoolEnabled:bool=True, Translation="0,0,0", Rotation="0,-50,0", Scale="1,1,1") -> (Pose) {
    alphaInputType=Bool
    rotationMode=BMM_Additive
    rotationSpace=BCS_BoneSpace
    translationMode=BMM_Additive
    translationSpace=BCS_BoneSpace
    alphaBoolBlend {
     blendInTime=0.1
     blendOption=Linear
     blendOutTime=1
     customCurve=&M_MED_OliveStomp_Curve
    }
    boneToModify {boneName=dyn_paper_02}
    rotation {yaw=-50}
   }
   n25 = "Transform (Modify) Bone - Bone: pack"(ComponentPose=@n24, Translation="0,-1,2", Rotation="0,0,0") {
    alphaInputType=Bool
    translationMode=BMM_Additive
    translationSpace=BCS_BoneSpace
    alphaBoolBlend {
     blendInTime=0.1
     blendOption=Linear
     blendOutTime=1
     customCurve=&M_MED_OliveStomp_Curve
    }
    boneToModify {boneName=pack}
    translation {y=-1, z=2}
   }
   n26 = "Transform (Modify) Bone - Bone: dyn_paper_02"(ComponentPose=@n25, bAlphaBoolEnabled=True, Translation="0,0,0", Rotation="0,50,0", Scale="1,1,1") {
    alphaInputType=Bool
    rotationMode=BMM_Additive
    rotationSpace=BCS_BoneSpace
    translationMode=BMM_Additive
    translationSpace=BCS_BoneSpace
    alphaBoolBlend {
     blendInTime=3
     blendOption=CircularIn
     blendOutTime=0.1
     customCurve=&M_MED_OliveStomp_Curve
    }
    boneToModify {boneName=dyn_paper_02}
    rotation {yaw=50}
   }
   n27 = "Transform (Modify) Bone - Bone: pack"(ComponentPose=@n26, Translation="0,1,-2", Rotation="0,0,0") {
    alphaInputType=Bool
    translationMode=BMM_Additive
    translationSpace=BCS_BoneSpace
    alphaBoolBlend {
     blendInTime=3
     blendOption=CircularIn
     blendOutTime=0.1
     customCurve=&M_MED_OliveStomp_Curve
    }
    boneToModify {boneName=pack}
    translation {y=1, z=-2}
   }
   n28 = "Transform (Modify) Bone - Bone: dyn_L_corner"!(ComponentPose=@n27, bAlphaBoolEnabled:bool=True, Translation="0,0,0", Rotation="0,0,-15", Scale="1,1,1") -> (Pose) {
    alphaInputType=Bool
    rotationMode=BMM_Additive
    rotationSpace=BCS_BoneSpace
    translationMode=BMM_Additive
    translationSpace=BCS_BoneSpace
    alphaBoolBlend {blendInTime=0.3, blendOption=Sinusoidal, blendOutTime=0.3}
    boneToModify {boneName=dyn_L_corner}
    rotation {roll=-15}
   }
   n29 = "Transform (Modify) Bone - Bone: dyn_paper_base"(ComponentPose=@n28, Alpha=Get On Melee Right Extra(), Translation="0,0,0", Rotation="10,-10,0", Scale="1,1,1") {
    rotationMode=BMM_Additive
    rotationSpace=BCS_BoneSpace
    scaleMode=BMM_Additive
    scaleSpace=BCS_BoneSpace
    translationMode=BMM_Additive
    translationSpace=BCS_BoneSpace
    alphaBoolBlend {blendOption=Linear}
    alphaScaleBiasClamp {bInterpResult=true, interpSpeedDecreasing=5, interpSpeedIncreasing=5}
    boneToModify {boneName=dyn_paper_base}
    rotation {pitch=10, yaw=-10}
   }
   n30 = "Transform (Modify) Bone - Bone: dyn_L_corner"!(ComponentPose=@n29, Alpha:float=1, Translation="0,0,0", Rotation="0,0,-25", Scale="1,1,1") -> (Pose) {
    rotationMode=BMM_Additive
    rotationSpace=BCS_BoneSpace
    translationMode=BMM_Additive
    translationSpace=BCS_BoneSpace
    alphaBoolBlend {blendInTime=0.3, blendOption=Linear, blendOutTime=0.3}
    alphaScaleBiasClamp {bInterpResult=true, interpSpeedDecreasing=5, interpSpeedIncreasing=5}
    boneToModify {boneName=dyn_L_corner}
    rotation {roll=-25}
   }
   n31 = "Transform (Modify) Bone - Bone: dyn_R_corner"(ComponentPose=@n30, Alpha=1, Translation="0,0,0", Rotation="0,0,45", Scale="1,1,1") {
    rotationMode=BMM_Additive
    rotationSpace=BCS_BoneSpace
    scaleMode=BMM_Additive
    scaleSpace=BCS_BoneSpace
    translationMode=BMM_Additive
    translationSpace=BCS_BoneSpace
    alphaBoolBlend {blendOption=Linear}
    alphaScaleBiasClamp {bInterpResult=true, interpSpeedDecreasing=5, interpSpeedIncreasing=5}
    boneToModify {boneName=dyn_R_corner}
    rotation {roll=45}
   }
   n32 = "Transform (Modify) Bone - Bone: dyn_paper_02"!(ComponentPose=@n31, Alpha:float=1, Translation="0,0,0", Rotation="0,-19,0", Scale="1,1,1") -> (Pose) {
    rotationMode=BMM_Additive
    rotationSpace=BCS_BoneSpace
    scaleMode=BMM_Additive
    scaleSpace=BCS_BoneSpace
    translationMode=BMM_Additive
    translationSpace=BCS_BoneSpace
    alphaBoolBlend {blendOption=Linear}
    boneToModify {boneName=dyn_paper_02}
    rotation {yaw=-19}
   }
   n33 = "Transform (Modify) Bone - Bone: dyn_paper_01"(ComponentPose=@n32, Alpha=1, Translation="0,0,0", Rotation="0,-5,0", Scale="1,1,1") {
    rotationMode=BMM_Additive
    rotationSpace=BCS_BoneSpace
    alphaBoolBlend {blendOption=Linear}
    boneToModify {boneName=dyn_paper_01}
    rotation {yaw=-5}
   }
   n34 = "Transform (Modify) Bone - Bone: dyn_paper_01"(ComponentPose=@n33, Alpha=1, Translation="0,0,0", Rotation="0,-5,0", Scale="1,1,1") {
    rotationMode=BMM_Additive
    rotationSpace=BCS_BoneSpace
    alphaBoolBlend {blendOption=Linear}
    boneToModify {boneName=dyn_paper_01}
    rotation {yaw=-5}
   }
   Save cached pose 'BaseMovement'(Pose=Component To Local(ComponentPose=@n34)) {cacheName=BaseMovement}
   Save cached pose 'Base0'(Pose=Input Pose()) {cacheName=Base0}
  ]}
 EventGraph {nodes [
   Delay(Duration=0.125)
   > ResetDynamics(InTeleportType=TeleportPhysics)
   > ResetDynamics(InTeleportType=ResetPhysics)
   Event BlueprintInitializeAnimation()
   > Set SkelMesh(SkelMesh=GetOwnerSkeletalMeshComponent())
   n1 = TryGetPawnOwner()
   > Is Valid(InputObject=@n1)
   > Set PlayerPawn(PlayerPawn:Pawn Object Reference=@n1) -> (Output_Get:Pawn Object Reference)
   n2 = Event OnCharacterCustomizationLoaded()
   > Set Base Mesh(Base Mesh:Skeletal Mesh Component Object Reference=@n2.CustomBaseMesh) -> (Output_Get:Skeletal Mesh Component Object Reference)
   n3 = GetPartSkeletalMeshComponent()
   > n4 = Get Dynamic Material Instance(SkeletalMeshComponent=@n3)
   > Set Mat Instance Dynamic(Mat Instance Dynamic=@n4)
   > n5 = GetOrCreateDynamicOverlayMaterial(MeshComponent=@n3, MaterialName=None)
   > Set Overlay Mat Dynamic(Overlay Mat Dynamic:Material Instance Dynamic Object Reference=@n5) -> (Output_Get:Material Instance Dynamic Object Reference)
   Event BlueprintUpdateAnimation()
   > n6 = Sequence()
   n7 = Get bIsParachuteOpened()
   > n8 = Branch(Condition=AND Boolean(A=Get bIsSkyDiving(), B=NOT Boolean(A=@n7)))
   n9 = GetRealTimeSeconds()
   n10 = float * float(A=1.36, B=0.75)
   n11 = float * float(A="Sin (Radians)"(A=float * float(A=@n10, B=@n9)), B=0.8)
   n12 = float * float(A=-0.55, B=1.25)
   n13 = float * float(A="Sin (Radians)"(A=float * float(A=@n12, B=@n9)), B=1.2)
   > Set Dyn_pack_base(Dyn_pack_base:Vector=MakeVector(X=@n11, Y=-2.5, Z=@n13)) -> (Output_Get:Vector)
   Set Dyn_pack_base(execute=@n8.else, Dyn_pack_base=MakeVector(X=@n11, Y=0, Z=@n13))
   n14 = Sequence!(execute=@n6.then_1) -> (then_0, then_1, then_2:Exec, then_3:Exec)
   > n15 = Gravity Override State Function DEPRECATED(
    CharacterPartAnimInstance=Self-Reference()
    GameTimeInSeconds=Get Game Time in Seconds() -> (Game Time in Seconds:double)
    GravityOverrideParams=Get Gravity Override Params()
   )
   > Set Gravity Override(Gravity Override=@n15)
   n16 = GetOwnerSkeletalMeshComponent()
   n17 = BreakRotator(InRot=BreakTransform(InTransform=GetDeltaTransformFromRefPose(self=@n16, BoneName=spine_05, BaseName=spine_01)).Rotation)
   n18 = MapRangeClamped(Value=@n17, InRangeA=20, InRangeB=46, OutRangeB=5)
   > n19 = Set Backpack_Rotate(Backpack_Rotate:Rotator=MakeRotator(Roll=MapRangeClamped(Value=@n17, InRangeA=-20, InRangeB=-46, OutRangeB=10), Yaw=@n18)) -> (Output_Get:Rotator)
   n20 = MapRangeClamped(Value=@n17, InRangeA=-20, InRangeB=-46, OutRangeB=-2)
   n21 = MapRangeClamped(Value=@n17, InRangeA=8, InRangeB=46, OutRangeB=-2)
   n22 = MapRangeClamped(Value=@n17.Yaw, InRangeA=25, InRangeB=35, OutRangeB=-2)
   n23 = Set PelvisLocation(execute=@n24.then, PelvisLocation:Vector=MakeVector(Y=float + float!(A=@n21, B=@n20, C:double=@n22) -> (ReturnValue))) -> (Output_Get:Vector)
   n24 = PrintString!(
    execute=@n19.then
    InString="To String (Vector)"!(InVec:Vector=@n23) -> (ReturnValue:String)
    bPrintToScreen
    bPrintToLog
    TextColor=(R=0.1311,G=1,B=0,A=1)
    Duration=0
    Key=3
   ) -> (then)
   Is Valid(exec=@n14.then_1, InputObject=Get PlayerPawn() -> (PlayerPawn:Pawn Object Reference))
   > Set WPO Globals Function(
    Material=Get Mat Instance Dynamic()
    Overlay Material=Get Overlay Mat Dynamic() -> (Overlay Mat Dynamic:Material Instance Dynamic Object Reference)
    Skeletal Mesh=Get SkelMesh()
    "Local Noise Transform  Location"="Get Local Noise Transform  Location"() -> ("Local Noise Transform  Location":Vector)
    Local Noise Offset=Get Local Noise Offset()
   )
   > Set WPO Regional Function!(
    execute
    Material:Material Instance Dynamic Object Reference=Get Mat Instance Dynamic()
    Overlay Material:Material Instance Dynamic Object Reference=Get Overlay Mat Dynamic()
    Skeletal Mesh:Skeletal Mesh Component Object Reference=Get SkelMesh()
    Velocity Transform:Name=pelvis
    Velocity Interp Speed:double=Get Velocity Interp Speed()
    Collision Transform:Name=spine_03
    Collision Transform Axis:EAxis Enum
    Invert Collision Transform Axis:bool=false
    Delta Time X:double=0
    Previous Velocity Transform Location:"Vector (by ref)"=Get R1_Previous Velocity Transform Location() -> (R1_Previous Velocity Transform Location:Vector)
    Previous Velocity Transform Velocity:"Vector (by ref)"=Get R1_Previous Velocity Transform Velocity() -> (R1_Previous Velocity Transform Velocity:Vector)
    Region:int=1
   ) -> (then)
   n25 = Get SkelMesh()
   n26 = Get bIsPlayingMeleeAnim()
   n27 = Set onMelee_left(
    execute=@n14.then_2
    onMelee_left:double=SelectFloat(
     A=GetAlphaFromReferenceTransformDelta(SkeletalMeshComponent=@n25, BoneName=upperarm_l, BaseName=spine_05, DeltaComponentType=EulerX, InputRange=(X=0,Y=-38), TwistAxis="1,0,0")
     B=0
     bPickA=@n26
    )
   ) -> (Output_Get:double)
   > PrintString(InString="To String (Float)"(InDouble=@n27), Duration=0, Key=1)
   > Set onMelee_right(onMelee_right:double=SelectFloat(
     A=GetAlphaFromReferenceTransformDelta(SkeletalMeshComponent=@n25, BoneName=upperarm_r, BaseName=spine_05, DeltaComponentType=EulerX, InputRange=(X=0,Y=-38), TwistAxis="1,0,0")
     B=0
     bPickA=@n26
    )) -> (Output_Get:double)
   n28 = GetAlphaFromReferenceTransformDelta(SkeletalMeshComponent=@n25, BoneName=upperarm_r, BaseName=spine_05, DeltaComponentType=EulerX, InputRange=(X=0,Y=-35), TwistAxis="1,0,0")
   n29 = Get bIsPlayingMeleeAnim()
   n30 = Get onMelee_left() -> (onMelee_left:double)
   > n31 = Set On Melee Right Extra(On Melee Right Extra:double=SelectFloat(A=SelectFloat(A=@n28, B=0, bPickA=@n29), B=0, bPickA="Equal (Float)"(A=@n30))) -> (Output_Get:double)
   > PrintString(InString="To String (Float)"(InDouble=@n31), TextColor=(R=1,G=0,B=0.0153,A=1), Duration=0, Key=2)
   n32 = NOT Boolean(A=Get bIsTargeting())
   > Set ChangingWeapon_pistol(ChangingWeapon_pistol:double=SelectFloat(
     A=GetAlphaFromReferenceTransformDelta(SkeletalMeshComponent=@n25, BoneName=clavicle_r, BaseName=spine_05, DeltaComponentType=EulerY, InputRange=(X=20,Y=32), TwistAxis="1,0,0")
     B=0
     bPickA=@n32
    )) -> (Output_Get:double)
   > Set ChangingWeapon(ChangingWeapon:double=SelectFloat(
     A=GetAlphaFromReferenceTransformDelta(SkeletalMeshComponent=@n25, BoneName=upperarm_r, BaseName=clavicle_r, DeltaComponentType=EulerX, InputRange=(X=0,Y=-55), TwistAxis="1,0,0")
     B=0
     bPickA=@n32
    )) -> (Output_Get:double)
   > Set ChangingWeapon_spine(ChangingWeapon_spine:double=SelectFloat(
     A=GetAlphaFromReferenceTransformDelta(SkeletalMeshComponent=@n25, BoneName=spine_05, BaseName=pelvis, DeltaComponentType=EulerX, InputRange=(X=-10,Y=-32), TwistAxis="1,0,0")
     B=0
     bPickA=@n32
    )) -> (Output_Get:double)
   > Set Targeting_lookDown(Targeting_lookDown:double=SelectFloat(
     A=GetAlphaFromReferenceTransformDelta(SkeletalMeshComponent=@n25, BoneName=spine_05, BaseName=pelvis, InputRange=(X=15,Y=25), TwistAxis="1,0,0")
     B=0
     bPickA=Get bIsTargeting()
    )) -> (Output_Get:double)
   > Set Targeting_lookUp(Targeting_lookUp:double=SelectFloat(
     A=GetAlphaFromReferenceTransformDelta(SkeletalMeshComponent=@n25, BoneName=spine_05, BaseName=pelvis, DeltaComponentType=EulerX, InputRange=(X=12,Y=8), TwistAxis="1,0,0")
     B=0
     bPickA=Get bIsTargeting()
    )) -> (Output_Get:double)
   TryGetPawnOwner()
   "To String (Float)"(InDouble=@n28.RawOutput)
   "To String (Rotator)"!(InRot:Rotator=@n19) -> (ReturnValue:String)
   "To String (Float)"(InDouble=@n18)
   n33 = float * float(A=1.11, B=1.11)
   float * float(A="Sin (Radians)"(A=float * float(A=@n33, B=@n9)), B=50)
   n34 = Normalize!(A:Vector=Get PawnVelocity(), Tolerance:float=0.0001) -> (ReturnValue:Vector) {is=title K2Node_CallFunction}
   n35 = GetActorForwardVector!(self:Actor Object Reference=TryGetPawnOwner()) -> (ReturnValue:Vector)
   Set MovingForward(
    execute=@n37.then
    MovingForward:double=MapRangeClamped(Value=Dot Product!(A:Vector=@n35, B:Vector=@n34) -> (ReturnValue:double), InRangeA=-0.6, InRangeB=1)
   ) -> (Output_Get:double)
   n36 = OR Boolean(A=Get bIsJumping(), B=Get bIsFalling())
   n37 = "Set Jump&&Fall"(execute=@n23.then, "Jump&&Fall":bool=@n36) -> (Output_Get:bool)
  ]}
]
variables [
 Backpack_Rotate {type=Rotator}
 Base Mesh {type=SkeletalMeshComponent}
 ChangingWeapon {type=number}
 ChangingWeapon_pistol {type=number}
 ChangingWeapon_spine {type=number}
 Dyn_pack_base {type=Vector}
 Game Time in Seconds {type=number}
 Gravity Override {type=Vector}
 Gravity Override Params {type=GravityOverrideParamsStruct}
 "Jump&&Fall" {type=boolean}
 Local Noise Offset {type=Vector}
 "Local Noise Transform  Location" {type=Vector}
 Mat Instance Dynamic {type=MaterialInstanceDynamic}
 MovingForward {type=number}
 On Melee Right Extra {type=number}
 onMelee_left {type=number}
 onMelee_right {type=number}
 Overlay Mat Dynamic {type=MaterialInstanceDynamic}
 PelvisLocation {type=Vector}
 PlayerPawn {type=Pawn, default=null}
 R1_Previous Velocity Transform Location {type=Vector}
 R1_Previous Velocity Transform Velocity {type=Vector}
 SkelMesh {type=SkeletalMeshComponent}
 Targeting_lookDown {type=number}
 Targeting_lookUp {type=number}
 Velocity Interp Speed {type=number, default=1}
]
